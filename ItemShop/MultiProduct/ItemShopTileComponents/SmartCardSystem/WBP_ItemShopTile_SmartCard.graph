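# Item Shop smart card tile widget that shows either an item image or a "+N" number over a backplate.
# SetIsNumberTile flips a visibility switcher between the lazy image and the text, formatting the number when needed.
# SetCardSize sizes a square spacer that sets the tile's minimum size.
# SetCornerRadius drives rounded corners through dynamic material parameters on both backplate and item image, while PreConstruct enables the backplate material's switch parameter.
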WidgetBlueprint /ItemShop/MultiProduct/ItemShopTileComponents/SmartCardSystem/WBP_ItemShopTile_SmartCard
graphs [
 EventGraph {nodes [
   Event PreConstruct()
   > n1 = GetDynamicMaterial(self=Get Backplate() -> (Backplate:Image Object Reference))
   > SetScalarParameterValue(self=@n1, ParameterName=Backplate Switch, Value=1)
  ]}
 SetCardSize {nodes [
   n1 = SetCardSize!() -> (then, DesiredSize:int) {is=title K2Node_FunctionEntry}
   n2 = "To Float (Integer)"(InInt=@n1)
   > SetSize!(
    execute
    self:Spacer Object Reference=Get Spacer_MinSize() -> (Spacer_MinSize:Spacer Object Reference)
    InSize_X:double=@n2
    InSize_Y:double=@n2
   ) -> (then)
  ]}
 SetCornerRadius {nodes [
   SetCornerRadius!() -> (then, RadiusInPixels:float=~) {is=title K2Node_FunctionEntry}
   > n1 = Sequence()
   > n2 = GetDynamicMaterial(self=Get Backplate())
   > SetScalarParameterValue(self=@n2, ParameterName=Radius in Pixels, Value=1)
   n3 = Get RadiusInPixels() -> (RadiusInPixels:float)
   > SetVectorParameterValue!(execute, self=@n2, ParameterName=Corner Radius, Value_R:float=@n3, Value_G:float=@n3, Value_B:float=@n3, Value_A:float=@n3) -> (then)
   n4 = GetDynamicMaterial(execute=@n1.then_1, self=Get LazyImage_Item() -> (LazyImage_Item:Fort Lazy Image Object Reference))
   > n5 = Sequence!(execute) -> (then_0, then_1, then_2:Exec, then_3:Exec)
   n6 = Get RadiusInPixels()
   > SetScalarParameterValue(self=@n4, ParameterName=TopLeftCornerRadius, Value=@n6)
   SetScalarParameterValue(execute=@n5.then_1, self=@n4, ParameterName=TopRightCornerRadius, Value=@n6)
   SetScalarParameterValue(execute=@n5.then_2, self=@n4, ParameterName=BottomLeftCornerRadius, Value=@n6)
   SetScalarParameterValue(execute=@n5.then_3, self=@n4, ParameterName=BottomRightCornerRadius, Value=@n6)
  ]}
 SetIsNumberTile {nodes [
   SetIsNumberTile!() -> (then, IsNumberTile:bool=~, Number:int) {is=title K2Node_FunctionEntry}
   > SetActiveWidget(
    self=Get Switcher_TileType() -> (Switcher_TileType:Common Visibility Switcher Object Reference)
    Widget=Select(
     Option 0:Widget Object Reference=Get LazyImage_Item()
     Option 1:Widget Object Reference=Get Text_Number() -> (Text_Number:Common Text Object Reference)
     Index=Get IsNumberTile() -> (IsNumberTile:bool)
    ) -> (ReturnValue:Widget Object Reference)
   )
   > Branch(Condition=Get IsNumberTile())
   > Set Text(Text=Format Text!(Format="LOCTEXT(\"+{number}\")", number:int=Get Number() -> (Number:int=0)) -> (Result), self=Get Text_Number())
  ]}
]
variables []
widgets {
 info {inheritedWidgetCount=0, parentClass=&UserWidget, rootWidgetClass=&Overlay, widgetCount=6}
 widgets [
  Overlay_25 {widgetClassPath=&Overlay, widget=Overlay_25}
  Backplate {
   widgetClassPath=&Image
   bIsVariable=true
   parent=Overlay_25
   widget=Backplate
   properties {
    visibility=SelfHitTestInvisible
    brush {
     bIsDynamicallyLoaded=false
     drawAs=Image
     imageType=FullColor
     mirroring=NoMirror
     resourceName=None
     resourceObject=/FortUILibrary/Systems/Backplate/MaterialInstances/MI_UIKit_Backplate
     tiling=NoTile
     imageSize=<8 8>
     margin=<ltrb 0 0 0 0>
     outlineSettings {
      bUseBrushTransparency=false
      roundingType=HalfHeightRadius
      width=0
      color {colorUseRule=UseColor_Specified, specifiedColor=<rgba 0 0 0 0>}
      cornerRadii=<0 0 0 0>
     }
     tintColor {colorUseRule=UseColor_Specified, specifiedColor=<rgba 1 1 1 1>}
     uVRegion {bIsValid=false, max=<0 0>, min=<0 0>}
    }
   }
   slot {class=OverlaySlot, properties {horizontalAlignment=HAlign_Fill, verticalAlignment=VAlign_Fill}}
  }
  Switcher_TileType {
   widgetClassPath=&CommonVisibilitySwitcher
   bIsVariable=true
   parent=Overlay_25
   widget=Switcher_TileType
   slot {class=OverlaySlot, properties {horizontalAlignment=HAlign_Fill, verticalAlignment=VAlign_Fill}}
  }
  LazyImage_Item {
   widgetClassPath=&FortLazyImage
   bIsVariable=true
   parent=Switcher_TileType
   widget=LazyImage_Item
   properties {
    materialTextureParamName=Item
    brush {
     bIsDynamicallyLoaded=false
     drawAs=Image
     imageType=FullColor
     mirroring=NoMirror
     resourceName=None
     resourceObject=/FortUILibrary/Systems/Items/MaterialInstances/MI_UIKit_ItemLocker
     tiling=NoTile
     imageSize=<32 32>
     margin=<ltrb 0 0 0 0>
     outlineSettings {
      bUseBrushTransparency=false
      roundingType=HalfHeightRadius
      width=0
      color {colorUseRule=UseColor_Specified, specifiedColor=<rgba 0 0 0 0>}
      cornerRadii=<0 0 0 0>
     }
     tintColor {colorUseRule=UseColor_Specified, specifiedColor=<rgba 1 1 1 1>}
     uVRegion {bIsValid=false, max=<0 0>, min=<0 0>}
    }
   }
   slot {class=CommonVisibilitySwitcherSlot}
  }
  Text_Number {
   widgetClassPath=&CommonTextBlock
   bIsVariable=true
   parent=Switcher_TileType
   widget=Text_Number
   properties {
    text=+2
    font {
     bForceMonospaced=false
     bMaterialIsStencil=false
     fontMaterial=null
     fontObject=&HeadingNow
     letterSpacing=0
     monospacedWidth=1
     size=27
     skewAmount=0
     typefaceFontName=76 Bold
     outlineSettings {
      bApplyOutlineToDropShadows=false
      bMiteredCorners=false
      bSeparateFillAlpha=false
      outlineMaterial=null
      outlineSize=0
      outlineColor=<rgba 0 0 0 1>
     }
    }
    shadowOffset=<0 0>
   }
   slot {class=CommonVisibilitySwitcherSlot, properties {horizontalAlignment=HAlign_Center, verticalAlignment=VAlign_Center}}
  }
  Spacer_MinSize {
   widgetClassPath=&Spacer
   bIsVariable=true
   parent=Overlay_25
   widget=Spacer_MinSize
   properties {size=<96 96>}
   slot {class=OverlaySlot}
  }
 ]
}
